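# Default Lit material for foundations that maps the T_Foundations texture by world XY position, aligning it to the builder grid.
# The Builder Grid Size parameter times four sets the projection scale, so one tile spans four grid cells.
# A constant eighth-of-a-tile UV offset shifts the pattern by half a cell.
# The graph shows no connection from the texture sample to a material output.
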Material /Game/Environments/World/Foundations/Materials/M_Foundation_WorldCoords
graph {
 properties {pixelDepthOffsetMode=PDOM_Legacy, usedShadingModels=Default Lit}
 nodes [
  n1 = Multiply(A=ScalarParameter() {defaultValue=512, parameterName=Builder Grid Size}) {constB=4}
  TextureSample(UVs=Add(A=MaterialFunctionCall!(Scale=@n1) {materialFunction=/Engine/Functions/Engine_MaterialFunctions01/Coordinates/WorldPosition-XY}.Result) {constB=0.125}) {texture=/Game/Environments/World/Foundations/Textures/T_Foundations}
 ]
}
